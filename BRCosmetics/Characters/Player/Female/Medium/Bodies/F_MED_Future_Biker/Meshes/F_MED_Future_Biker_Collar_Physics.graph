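# Physics asset for the jacket collar of the female medium Future Biker body, letting collar flaps and straps move as constrained rigid bodies.
# Box bodies cover left, right and back flaps plus two straps, while capsules on the upper spine and neck approximate the body.
# Each flap has limited rotation against a tiny MSTR body, with straps hinged off the 01 flaps.
# Side MSTR bodies link to spine_05 with free rotation; the back MSTR has no such constraint.
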PhysicsAsset /BRCosmetics/Characters/Player/Female/Medium/Bodies/F_MED_Future_Biker/Meshes/F_MED_Future_Biker_Collar_Physics
bodies {
 dyn_collar_01_l [{shapeType=Box, extentX=5.8199, extentY=8.9263, extentZ=1.1, center=<2.7581 0 -0.0001>}]
 dyn_collar_01_l_MSTR [{shapeType=Box, extentX=0.01, extentY=0.01, extentZ=0.01}]
 dyn_collar_01_r [{shapeType=Box, extentX=5.8199, extentY=8.9263, extentZ=1.1, center=<-2.7581 0 0.0001>, rotation=<rot 0 0 180>}]
 dyn_collar_01_r_MSTR [{shapeType=Box, extentX=0.01, extentY=0.01, extentZ=0.01}]
 dyn_collar_03_l [{shapeType=Box, extentX=5.8199, extentY=8.9263, extentZ=1.1, center=<2.7581 0 -0.0001>}]
 dyn_collar_03_l_MSTR [{shapeType=Box, extentX=0.01, extentY=0.01, extentZ=0.01}]
 dyn_collar_03_r [{shapeType=Box, extentX=5.8199, extentY=8.9263, extentZ=1.1, center=<-2.7581 0 0.0001>, rotation=<rot 0 0 180>}]
 dyn_collar_03_r_MSTR [{shapeType=Box, extentX=0.01, extentY=0.01, extentZ=0.01}]
 dyn_collar_back_03 [{shapeType=Box, extentX=5.8199, extentY=6.4263, extentZ=1.1, center=<2.7581 0 -0.0001>}]
 dyn_collar_back_03_MSTR [{shapeType=Box, extentX=0.01, extentY=0.01, extentZ=0.01}]
 dyn_collar_strap_01_l [{shapeType=Box, extentX=5.3199, extentY=1.4263, extentZ=4, center=<2.7899 0.0001 -0.0002>}]
 dyn_collar_strap_01_r [{shapeType=Box, extentX=5.3199, extentY=1.4263, extentZ=4, center=<-2.7899 -0.0001 0.0002>, rotation=<rot 0 0 180>}]
 neck_01 [{shapeType=Capsule, length=1.75, radius=4.4657, center=<2.7969 0.0001 0>, rotation=<rot -90 -2.5305 2.5304>}]
 neck_02 [{shapeType=Capsule, length=1.75, radius=4.9657, center=<2.7969 0.0001 0>, rotation=<rot -90 -2.5305 2.5304>}]
 spine_05 [{shapeType=Capsule, length=2.5, radius=10.35, center=<0 1.9402 0>}]
}
constraints [
 {bone1Name=dyn_collar_strap_01_l, bone2Name=dyn_collar_01_l, swing1LimitDegrees=15, swing2LimitDegrees=15, twistLimitDegrees=5}
 {
  bone1Name=dyn_collar_01_l_MSTR
  bone2Name=spine_05
  swing1LimitDegrees=45
  swing1Motion=Free
  swing2LimitDegrees=45
  swing2Motion=Free
  twistLimitDegrees=45
  twistMotion=Free
 }
 {
  bone1Name=dyn_collar_03_l_MSTR
  bone2Name=spine_05
  swing1LimitDegrees=45
  swing1Motion=Free
  swing2LimitDegrees=45
  swing2Motion=Free
  twistLimitDegrees=45
  twistMotion=Free
 }
 {
  bone1Name=dyn_collar_03_r_MSTR
  bone2Name=spine_05
  swing1LimitDegrees=45
  swing1Motion=Free
  swing2LimitDegrees=45
  swing2Motion=Free
  twistLimitDegrees=45
  twistMotion=Free
 }
 {
  bone1Name=dyn_collar_01_r_MSTR
  bone2Name=spine_05
  swing1LimitDegrees=45
  swing1Motion=Free
  swing2LimitDegrees=45
  swing2Motion=Free
  twistLimitDegrees=45
  twistMotion=Free
 }
 {bone1Name=dyn_collar_back_03, bone2Name=dyn_collar_back_03_MSTR, swing1LimitDegrees=5, swing2LimitDegrees=65, twistLimitDegrees=5}
 {bone1Name=dyn_collar_03_l, bone2Name=dyn_collar_03_l_MSTR, swing1LimitDegrees=5, swing2LimitDegrees=45, twistLimitDegrees=5}
 {
  bone1Name=neck_01
  bone2Name=spine_05
  swing1LimitDegrees=45
  swing1Motion=Free
  swing2LimitDegrees=45
  swing2Motion=Free
  twistLimitDegrees=45
  twistMotion=Free
 }
 {
  bone1Name=neck_02
  bone2Name=neck_01
  swing1LimitDegrees=45
  swing1Motion=Free
  swing2LimitDegrees=45
  swing2Motion=Free
  twistLimitDegrees=45
  twistMotion=Free
 }
 {bone1Name=dyn_collar_01_l, bone2Name=dyn_collar_01_l_MSTR, swing1LimitDegrees=45, swing2LimitDegrees=45, twistLimitDegrees=5}
 {bone1Name=dyn_collar_strap_01_r, bone2Name=dyn_collar_01_r, swing1LimitDegrees=15, swing2LimitDegrees=15, twistLimitDegrees=5}
 {bone1Name=dyn_collar_03_r, bone2Name=dyn_collar_03_r_MSTR, swing1LimitDegrees=5, swing2LimitDegrees=65, twistLimitDegrees=5}
 {bone1Name=dyn_collar_01_r, bone2Name=dyn_collar_01_r_MSTR, swing1LimitDegrees=45, swing2LimitDegrees=45, twistLimitDegrees=5}
]
