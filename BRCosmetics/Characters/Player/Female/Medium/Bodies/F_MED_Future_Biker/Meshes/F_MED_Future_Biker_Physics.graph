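# Physics asset for the female medium Future Biker player body, set up for secondary motion of jacket, back shirt flap and left-side cannister.
# Pelvis and spine capsules form the body core, with tiny _MSTR anchor boxes joined to them by free-rotation constraints.
# From each anchor hang one- or two-link box or capsule chains with angular limits.
# Jacket back locks twist, while the shirt swings on one axis only.
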PhysicsAsset /BRCosmetics/Characters/Player/Female/Medium/Bodies/F_MED_Future_Biker/Meshes/F_MED_Future_Biker_Physics
bodies {
 dyn_cannister_01_l [{shapeType=Capsule, length=11.7709, radius=3.3847, center=<3.6541 0.076 -2.113>, rotation=<rot 0 90.0565 -91.0489>}]
 dyn_cannister_01_l_MSTR [{shapeType=Box, extentX=0.01, extentY=0.01, extentZ=0.01}]
 dyn_jacket_back_01 [{shapeType=Box, extentX=4.8199, extentY=7.4263, extentZ=4.6321, center=<2.4 0 0.8>}]
 dyn_jacket_back_01_MSTR [{shapeType=Box, extentX=0.01, extentY=0.01, extentZ=0.01}]
 dyn_jacket_back_02 [{shapeType=Box, extentX=4.8199, extentY=7.4263, extentZ=1.6, center=<2.4 0 0.8>}]
 dyn_jacket_front_01_l [{shapeType=Box, extentX=4.8199, extentY=7.4263, extentZ=1.1, center=<2.3497 0 -0.0002>}]
 dyn_jacket_front_01_l_MSTR [{shapeType=Box, extentX=0.01, extentY=0.01, extentZ=0.01}]
 dyn_jacket_front_01_r [{shapeType=Box, extentX=4.8199, extentY=7.4263, extentZ=1.1, center=<-2.3497 0 0.0002>, rotation=<rot 0 0 180>}]
 dyn_jacket_front_01_r_MSTR [{shapeType=Box, extentX=0.01, extentY=0.01, extentZ=0.01}]
 dyn_jacket_front_02_l [{shapeType=Box, extentX=4.8199, extentY=7.4263, extentZ=1.1, center=<2.3497 0 -0.0002>}]
 dyn_jacket_front_02_r [{shapeType=Box, extentX=4.8199, extentY=7.4263, extentZ=1.1, center=<2.3229 0 0.0003>, rotation=<rot 0 0 -180>}]
 dyn_jacket_mid_01_l [{shapeType=Box, extentX=2.3199, extentY=2.9263, extentZ=1.1, center=<1.204 0 -0.0002>}]
 dyn_jacket_mid_01_l_MSTR [{shapeType=Box, extentX=0.01, extentY=0.01, extentZ=0.01}]
 dyn_jacket_mid_01_r [{shapeType=Box, extentX=2.3199, extentY=2.9263, extentZ=1.1, center=<-1.204 0 0.0002>, rotation=<rot 0 0 180>}]
 dyn_jacket_mid_01_r_MSTR [{shapeType=Box, extentX=0.01, extentY=0.01, extentZ=0.01}]
 dyn_jacket_side_01_l [{shapeType=Box, extentX=5, extentY=7.4, extentZ=4.6028, center=<2.3497 0 0>}]
 dyn_jacket_side_01_l_MSTR [{shapeType=Box, extentX=0.01, extentY=0.01, extentZ=0.01}]
 dyn_jacket_side_01_r [{shapeType=Box, extentX=5, extentY=7.4, extentZ=4.6028, center=<-2.3497 0 0>, rotation=<rot 0 0 180>}]
 dyn_jacket_side_01_r_MSTR [{shapeType=Box, extentX=0.01, extentY=0.01, extentZ=0.01}]
 dyn_jacket_side_02_l [{shapeType=Box, extentX=5, extentY=7.4, extentZ=3.6, center=<2.3497 0 -0.0002>}]
 dyn_jacket_side_02_r [{shapeType=Box, extentX=5, extentY=7.4, extentZ=3.6, center=<-2.3497 0 0.0002>, rotation=<rot 0 0 180>}]
 dyn_jacket_upper_01_l [{shapeType=Box, extentX=2.3199, extentY=5.4263, extentZ=1.1, center=<1.204 0 -0.0002>}]
 dyn_jacket_upper_01_l_MSTR [{shapeType=Box, extentX=0.01, extentY=0.01, extentZ=0.01}]
 dyn_jacket_upper_01_r [{shapeType=Box, extentX=2.3199, extentY=5.4263, extentZ=1.1, center=<-1.204 0 0.0002>, rotation=<rot 0 0 180>}]
 dyn_jacket_upper_01_r_MSTR [{shapeType=Box, extentX=0.01, extentY=0.01, extentZ=0.01}]
 dyn_shirt_back_01 [{shapeType=Box, extentX=1.1, extentY=13, extentZ=4.3, center=<2.3 0 0.4>, rotation=<rot 90 0 0>}]
 dyn_shirt_back_01_MSTR [{shapeType=Box, extentX=0.01, extentY=0.01, extentZ=0.01}]
 dyn_shirt_back_02 [{shapeType=Box, extentX=1.9081, extentY=8.5851, extentZ=4.3, center=<2.3 0 0.4>, rotation=<rot 90 0 0>}]
 pelvis [{shapeType=Capsule, length=7.204, radius=9.9885, center=<0 -1.2137 -0.0001>}]
 spine_02 [{shapeType=Capsule, length=1.25, radius=8.85, center=<0 0.9735 0>}]
 spine_03 [{shapeType=Capsule, length=2.75, radius=9.1, center=<0 2.2385 -0.0001>}]
 spine_04 [{shapeType=Capsule, length=1.5, radius=10.1, center=<0 4.2244 0.0001>}]
 spine_05 [{shapeType=Capsule, length=2.5, radius=10.35, center=<0 1.9402 0>}]
}
constraints [
 {
  bone1Name=dyn_jacket_upper_01_l_MSTR
  bone2Name=spine_05
  swing1LimitDegrees=45
  swing1Motion=Free
  swing2LimitDegrees=45
  swing2Motion=Free
  twistLimitDegrees=45
  twistMotion=Free
 }
 {
  bone1Name=dyn_jacket_upper_01_r_MSTR
  bone2Name=spine_05
  swing1LimitDegrees=45
  swing1Motion=Free
  swing2LimitDegrees=45
  swing2Motion=Free
  twistLimitDegrees=45
  twistMotion=Free
 }
 {
  bone1Name=dyn_jacket_mid_01_l_MSTR
  bone2Name=spine_04
  swing1LimitDegrees=45
  swing1Motion=Free
  swing2LimitDegrees=45
  swing2Motion=Free
  twistLimitDegrees=45
  twistMotion=Free
 }
 {
  bone1Name=dyn_jacket_mid_01_r_MSTR
  bone2Name=spine_04
  swing1LimitDegrees=45
  swing1Motion=Free
  swing2LimitDegrees=45
  swing2Motion=Free
  twistLimitDegrees=45
  twistMotion=Free
 }
 {
  bone1Name=dyn_jacket_front_01_l_MSTR
  bone2Name=spine_04
  swing1LimitDegrees=45
  swing1Motion=Free
  swing2LimitDegrees=45
  swing2Motion=Free
  twistLimitDegrees=45
  twistMotion=Free
 }
 {
  bone1Name=dyn_jacket_front_01_r_MSTR
  bone2Name=spine_04
  swing1LimitDegrees=45
  swing1Motion=Free
  swing2LimitDegrees=45
  swing2Motion=Free
  twistLimitDegrees=45
  twistMotion=Free
 }
 {
  bone1Name=dyn_jacket_side_01_l_MSTR
  bone2Name=spine_04
  swing1LimitDegrees=45
  swing1Motion=Free
  swing2LimitDegrees=45
  swing2Motion=Free
  twistLimitDegrees=45
  twistMotion=Free
 }
 {
  bone1Name=dyn_jacket_back_01_MSTR
  bone2Name=spine_04
  swing1LimitDegrees=45
  swing1Motion=Free
  swing2LimitDegrees=45
  swing2Motion=Free
  twistLimitDegrees=45
  twistMotion=Free
 }
 {
  bone1Name=dyn_jacket_side_01_r_MSTR
  bone2Name=spine_04
  swing1LimitDegrees=45
  swing1Motion=Free
  swing2LimitDegrees=45
  swing2Motion=Free
  twistLimitDegrees=45
  twistMotion=Free
 }
 {
  bone1Name=dyn_cannister_01_l_MSTR
  bone2Name=pelvis
  swing1LimitDegrees=45
  swing1Motion=Free
  swing2LimitDegrees=45
  swing2Motion=Free
  twistLimitDegrees=45
  twistMotion=Free
 }
 {
  bone1Name=dyn_shirt_back_01_MSTR
  bone2Name=pelvis
  swing1LimitDegrees=45
  swing1Motion=Free
  swing2LimitDegrees=45
  swing2Motion=Free
  twistLimitDegrees=45
  twistMotion=Free
 }
 {bone1Name=dyn_jacket_mid_01_l, bone2Name=dyn_jacket_mid_01_l_MSTR, swing1LimitDegrees=5, swing2LimitDegrees=45, twistLimitDegrees=5}
 {bone1Name=dyn_jacket_front_01_l, bone2Name=dyn_jacket_front_01_l_MSTR, swing1LimitDegrees=30, swing2LimitDegrees=30, twistLimitDegrees=5}
 {bone1Name=dyn_jacket_front_02_l, bone2Name=dyn_jacket_front_01_l, swing1LimitDegrees=30, swing2LimitDegrees=30, twistLimitDegrees=5}
 {bone1Name=dyn_jacket_front_02_r, bone2Name=dyn_jacket_front_01_r, swing1LimitDegrees=30, swing2LimitDegrees=30, twistLimitDegrees=5}
 {bone1Name=dyn_jacket_front_01_r, bone2Name=dyn_jacket_front_01_r_MSTR, swing1LimitDegrees=30, swing2LimitDegrees=30, twistLimitDegrees=5}
 {bone1Name=dyn_jacket_side_01_l, bone2Name=dyn_jacket_side_01_l_MSTR, swing1LimitDegrees=30, swing2LimitDegrees=30, twistLimitDegrees=1}
 {bone1Name=dyn_jacket_side_02_l, bone2Name=dyn_jacket_side_01_l, swing1LimitDegrees=30, swing2LimitDegrees=30, twistLimitDegrees=1}
 {
  bone1Name=dyn_jacket_back_01
  bone2Name=dyn_jacket_back_01_MSTR
  swing1LimitDegrees=30
  swing2LimitDegrees=30
  twistLimitDegrees=5
  twistMotion=Locked
 }
 {
  bone1Name=dyn_jacket_back_02
  bone2Name=dyn_jacket_back_01
  swing1LimitDegrees=30
  swing2LimitDegrees=30
  twistLimitDegrees=5
  twistMotion=Locked
 }
 {
  bone1Name=spine_02
  bone2Name=pelvis
  swing1LimitDegrees=45
  swing1Motion=Free
  swing2LimitDegrees=45
  swing2Motion=Free
  twistLimitDegrees=45
  twistMotion=Free
 }
 {
  bone1Name=spine_03
  bone2Name=spine_02
  swing1LimitDegrees=45
  swing1Motion=Free
  swing2LimitDegrees=45
  swing2Motion=Free
  twistLimitDegrees=45
  twistMotion=Free
 }
 {
  bone1Name=spine_04
  bone2Name=spine_03
  swing1LimitDegrees=45
  swing1Motion=Free
  swing2LimitDegrees=45
  swing2Motion=Free
  twistLimitDegrees=45
  twistMotion=Free
 }
 {
  bone1Name=spine_05
  bone2Name=spine_04
  swing1LimitDegrees=45
  swing1Motion=Free
  swing2LimitDegrees=45
  swing2Motion=Free
  twistLimitDegrees=45
  twistMotion=Free
 }
 {bone1Name=dyn_cannister_01_l, bone2Name=dyn_cannister_01_l_MSTR, swing1LimitDegrees=45, swing2LimitDegrees=45, twistLimitDegrees=5}
 {bone1Name=dyn_jacket_upper_01_r, bone2Name=dyn_jacket_upper_01_r_MSTR, swing1LimitDegrees=15, swing2LimitDegrees=45, twistLimitDegrees=5}
 {bone1Name=dyn_jacket_upper_01_l, bone2Name=dyn_jacket_upper_01_l_MSTR, swing1LimitDegrees=15, swing2LimitDegrees=45, twistLimitDegrees=5}
 {bone1Name=dyn_jacket_mid_01_r, bone2Name=dyn_jacket_mid_01_r_MSTR, swing1LimitDegrees=5, swing2LimitDegrees=45, twistLimitDegrees=5}
 {
  bone1Name=dyn_shirt_back_01
  bone2Name=dyn_shirt_back_01_MSTR
  swing1LimitDegrees=45
  swing1Motion=Locked
  swing2LimitDegrees=45
  twistLimitDegrees=45
  twistMotion=Locked
 }
 {
  bone1Name=dyn_shirt_back_02
  bone2Name=dyn_shirt_back_01
  swing1LimitDegrees=45
  swing1Motion=Locked
  swing2LimitDegrees=45
  twistLimitDegrees=45
  twistMotion=Locked
 }
 {bone1Name=dyn_jacket_side_02_r, bone2Name=dyn_jacket_side_01_r, swing1LimitDegrees=30, swing2LimitDegrees=30, twistLimitDegrees=1}
 {bone1Name=dyn_jacket_side_01_r, bone2Name=dyn_jacket_side_01_r_MSTR, swing1LimitDegrees=30, swing2LimitDegrees=30, twistLimitDegrees=1}
]
