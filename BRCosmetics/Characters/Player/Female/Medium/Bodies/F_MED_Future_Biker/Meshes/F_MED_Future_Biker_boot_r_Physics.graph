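# Physics asset for the right boot of the female medium Future Biker character, setting up secondary motion for its dyn_boot bones.
# A thigh capsule anchors a tiny master box through a constraint with all angular axes free.
# Two identical boot boxes chain from that master through constraints with tight twist and first-swing limits and a wider second swing, so the chain bends mainly on one axis.
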PhysicsAsset /BRCosmetics/Characters/Player/Female/Medium/Bodies/F_MED_Future_Biker/Meshes/F_MED_Future_Biker_boot_r_Physics
bodies {
 dyn_boot_01_r [{shapeType=Box, extentX=4.5, extentY=11, extentZ=4.5, center=<2 0 0>}]
 dyn_boot_01_r_MSTR [{shapeType=Box, extentX=0.01, extentY=0.01, extentZ=0.01}]
 dyn_boot_02_r [{shapeType=Box, extentX=4.5, extentY=11, extentZ=4.5, center=<2 0 0>}]
 thigh_r [{shapeType=Capsule, length=37.1319, radius=6.7049, center=<20.3896 0.0004 -0.3422>, rotation=<rot -90 -0.0198 -179.9802>}]
}
constraints [
 {
  bone1Name=dyn_boot_01_r_MSTR
  bone2Name=thigh_r
  swing1LimitDegrees=45
  swing1Motion=Free
  swing2LimitDegrees=45
  swing2Motion=Free
  twistLimitDegrees=45
  twistMotion=Free
 }
 {bone1Name=dyn_boot_01_r, bone2Name=dyn_boot_01_r_MSTR, swing1LimitDegrees=5, swing2LimitDegrees=25, twistLimitDegrees=5}
 {bone1Name=dyn_boot_02_r, bone2Name=dyn_boot_01_r, swing1LimitDegrees=5, swing2LimitDegrees=25, twistLimitDegrees=5}
]
