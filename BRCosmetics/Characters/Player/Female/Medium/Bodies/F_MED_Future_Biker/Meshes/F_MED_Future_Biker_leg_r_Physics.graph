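# Physics asset for the right leg of the female medium Future Biker body, giving its dynamic leg strap physics bodies.
# Two box strap segments chain through a tiny master-bone box to the thigh capsule, with a calf capsule providing lower-leg shape.
# Strap joints lock twist, while the master-to-thigh and calf-to-thigh constraints leave all swing and twist free.
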PhysicsAsset /BRCosmetics/Characters/Player/Female/Medium/Bodies/F_MED_Future_Biker/Meshes/F_MED_Future_Biker_leg_r_Physics
bodies {
 calf_r [{shapeType=Capsule, length=21, radius=7, center=<14.5 4 0>, rotation=<rot -85 10 -180>}]
 dyn_leg_strap_01_r [{shapeType=Box, extentX=2, extentY=5, extentZ=5.5, center=<-0.2 0 0>, rotation=<rot 85 0 0>}]
 dyn_leg_strap_01_r_MSTR [{shapeType=Box, extentX=0.01, extentY=0.01, extentZ=0.01}]
 dyn_leg_strap_02_r [{shapeType=Box, extentX=1.7, extentY=5, extentZ=12.6, center=<-5.5 0 0>, rotation=<rot 90 0 0>}]
 thigh_r [{shapeType=Capsule, length=37.1319, radius=6.7049, center=<20.3896 0.0004 -0.3422>, rotation=<rot -90 -0.0198 -179.9802>}]
}
constraints [
 {
  bone1Name=dyn_leg_strap_02_r
  bone2Name=dyn_leg_strap_01_r
  swing1LimitDegrees=45
  swing2LimitDegrees=45
  twistLimitDegrees=45
  twistMotion=Locked
 }
 {
  bone1Name=dyn_leg_strap_01_r
  bone2Name=dyn_leg_strap_01_r_MSTR
  swing1LimitDegrees=45
  swing2LimitDegrees=45
  twistLimitDegrees=45
  twistMotion=Locked
 }
 {
  bone1Name=dyn_leg_strap_01_r_MSTR
  bone2Name=thigh_r
  swing1LimitDegrees=45
  swing1Motion=Free
  swing2LimitDegrees=45
  swing2Motion=Free
  twistLimitDegrees=45
  twistMotion=Free
 }
 {
  bone1Name=calf_r
  bone2Name=thigh_r
  swing1LimitDegrees=45
  swing1Motion=Free
  swing2LimitDegrees=45
  swing2Motion=Free
  twistLimitDegrees=45
  twistMotion=Free
 }
]
