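# Animation Blueprint for the female Future Biker body cosmetic, adding secondary motion and pose fixes for its jacket, collar, armor and boots.
# It copies the attached parent's pose, then chains head-driven collar constraints, RigidBody simulations, AnimDynamics zipper and upper-arm bones, corrective bone offsets and constraints.
# The event graph derives blend alphas from head and left-arm rotation, targeting, crouching and backpack presence, plus a recoil force on upper-arm bones after firing.
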AnimBlueprint /BRCosmetics/Characters/Player/Female/Medium/Bodies/F_MED_Future_Biker/Meshes/F_MED_Future_Biker_Skeleton_AnimBP
target_skeleton=&/F_MED_Future_Biker_Skeleton
graphs [
 AnimGraph {nodes [
   n1 = Get isAimingUpAlpha() -> (isAimingUpAlpha:double)
   n2 = Get headSwingAlpha() -> (headSwingAlpha:double)
   n3 = "Constraint - dyn_collar_back_03_MSTR (R)"(ComponentPose=Local To Component(LocalPose=Copy Pose From Mesh() {bUseAttachedParent=true}), Alpha=@n2, ConstraintWeights_0=1) {
    alphaBoolBlend {blendOption=Linear}
    boneToModify {boneName=dyn_collar_back_03_MSTR}
    constraintSetup [{transformType=Rotation, targetBone {boneName=head}}]
    constraintWeights [1]
   }
   n4 = "Constraint - dyn_collar_03_r_MSTR (R)"(ComponentPose=@n3, Alpha=@n2, ConstraintWeights_0=1) {
    alphaBoolBlend {blendOption=Linear}
    boneToModify {boneName=dyn_collar_03_r_MSTR}
    constraintSetup [{transformType=Rotation, targetBone {boneName=head}}]
    constraintWeights [1]
   }
   n5 = "Constraint - dyn_collar_03_l_MSTR (R)"(ComponentPose=@n4, Alpha=@n2, ConstraintWeights_0=1) {
    alphaBoolBlend {blendOption=Linear}
    boneToModify {boneName=dyn_collar_03_l_MSTR}
    constraintSetup [{transformType=Rotation, targetBone {boneName=head}}]
    constraintWeights [1]
   }
   n6 = "Transform (Modify) Bone - Bone: dyn_collar_back_03_MSTR"(ComponentPose=@n5, Alpha=@n2, Translation="0,0,0", Rotation="0,0,-15", Scale="1,1,1") {rotationMode=BMM_Additive, alphaBoolBlend {blendOption=Linear}, boneToModify {boneName=dyn_collar_back_03_MSTR}, rotation {roll=-15}}
   n7 = Break Fort Anim Input Skydiving External Force(FortAnimInput_SkydivingExternalForce=Get SkydivingExternalForceInput()) -> (
    bUseSkydivingVectorForce=~
    SkydivingVectorForce=~
    HeadToPelvisDirection=~
    FloatingMultiplier=~
    FloatingAdditive=~
    DivingMultiplier=~
    DivingAdditive=~
    ParachutingMultiplier=~
    ParachutingAdditive=~
   )
   n8 = RigidBody(ComponentPose=@n6, ExternalForce=@n7.SkydivingVectorForce) {
    bClampLinearTranslationLimitToRefPose=true
    bOverrideWorldGravity=true
    overridePhysicsAsset=&/F_MED_Future_Biker_Physics
    simulationSpace=BaseBoneSpace
    alphaBoolBlend {blendOption=Linear}
    baseBoneRef {boneName=spine_01}
    componentAppliedLinearAccClamp=<3000 3000 3000>
    componentLinearAccScale=<0.3 0.3 0.3>
    componentLinearVelScale=<0.3 0.3 0.3>
    overrideWorldGravity {z=-980}
   }
   n9 = RigidBody(ComponentPose=@n8, Alpha=Get isTargetingHalf() -> (isTargetingHalf:double), ExternalForce=@n7.SkydivingVectorForce) {
    bClampLinearTranslationLimitToRefPose=true
    bOverrideWorldGravity=true
    overridePhysicsAsset=&/F_MED_Future_Biker_Collar_Physics
    simulationSpace=BaseBoneSpace
    alphaBoolBlend {blendOption=Linear}
    baseBoneRef {boneName=spine_03}
    componentAppliedLinearAccClamp=<2500 2500 2500>
    componentLinearAccScale=<0.5 0.5 0.5>
    componentLinearVelScale=<0.7 0.7 0.7>
    overrideWorldGravity {z=-980}
   }
   n10 = RigidBody(ComponentPose=@n9, ExternalForce=@n7.SkydivingVectorForce) {
    bClampLinearTranslationLimitToRefPose=true
    bOverrideWorldGravity=true
    overridePhysicsAsset=&/F_MED_Future_Biker_leg_l_Physics
    simulationSpace=BaseBoneSpace
    alphaBoolBlend {blendOption=Linear}
    baseBoneRef {boneName=thigh_twist_01_l}
    componentAppliedLinearAccClamp=<3000 3000 3000>
    componentLinearAccScale=<0.3 0.3 0.7>
    componentLinearVelScale=<0.3 0.3 0.7>
    overrideWorldGravity {z=-980}
   }
   n11 = RigidBody(ComponentPose=@n10, ExternalForce=@n7.SkydivingVectorForce) {
    bClampLinearTranslationLimitToRefPose=true
    bOverrideWorldGravity=true
    overridePhysicsAsset=&/F_MED_Future_Biker_leg_r_Physics
    simulationSpace=BaseBoneSpace
    alphaBoolBlend {blendOption=Linear}
    baseBoneRef {boneName=thigh_twist_01_r}
    componentAppliedLinearAccClamp=<3000 3000 3000>
    componentLinearAccScale=<0.3 0.3 0.7>
    componentLinearVelScale=<0.3 0.3 0.7>
    overrideWorldGravity {z=-980}
   }
   n12 = RigidBody(ComponentPose=@n11, ExternalForce=@n7.SkydivingVectorForce) {
    bClampLinearTranslationLimitToRefPose=true
    bOverrideWorldGravity=true
    overridePhysicsAsset=&/F_MED_Future_Biker_boot_r_Physics
    simulationSpace=BaseBoneSpace
    alphaBoolBlend {blendOption=Linear}
    baseBoneRef {boneName=calf_twist_01_r}
    componentAppliedLinearAccClamp=<1000 1000 1000>
    componentLinearAccScale=<0.3 0.3 2>
    componentLinearVelScale=<0.6 0.6 2>
    overrideWorldGravity {z=-980}
   }
   n13 = Anim Dynamics - Bone: dyn_zipper_01(ComponentPose=@n12, Alpha=1, ExternalForce="0,0,0") {
    angularDampingOverride=0.999
    angularSpringConstant=350
    bAngularSpring=true
    bGravityOverrideInSimSpace=true
    bOverrideAngularDamping=true
    bUseGravityOverride=true
    bUsePlanarLimit=false
    gravityScale=0.25
    simulationSpace=RootRelative
    alphaBoolBlend {blendOption=Linear}
    boundBone {boneName=dyn_zipper_01}
    chainEnd {boneName=root}
    componentAppliedLinearAccClamp=<3000 3000 3000>
    componentLinearAccScale=<0.7 0.7 0.7>
    componentLinearVelScale=<0.7 0.7 0.7>
    gravityOverride {z=-980}
    physicsBodyDefinitions [{
      sphereCollisionRadius=1
      boundBone {boneName=dyn_zipper_01}
      boxExtents=<10 10 10>
      constraintSetup {angularLimitsMax=<0 15 25>, angularLimitsMin=<0 -15 0>, angularTarget=<0 0 0>}
      localJointOffset=<5 0 0>
     }]
    relativeSpaceBone {boneName=spine_03}
   }
   n14 = Get shootingForce() -> (shootingForce:Vector)
   n15 = Anim Dynamics - Bone: dyn_upperarm_02_l(ComponentPose=@n13, Alpha=1, ExternalForce=@n14) {
    angularDampingOverride=1
    angularSpringConstant=50
    bGravityOverrideInSimSpace=true
    bLinearSpring=true
    bOverrideAngularDamping=true
    bUseGravityOverride=true
    bUsePlanarLimit=false
    gravityScale=0.25
    linearSpringConstant=25
    simulationSpace=RootRelative
    alphaBoolBlend {blendOption=Linear}
    boundBone {boneName=dyn_upperarm_02_l}
    chainEnd {boneName=root}
    componentAppliedLinearAccClamp=<3000 3000 3000>
    componentLinearAccScale=<0.7 0.7 0.7>
    componentLinearVelScale=<0.7 0.7 0.7>
    gravityOverride {z=-980}
    physicsBodyDefinitions [{
      sphereCollisionRadius=1
      boundBone {boneName=dyn_upperarm_02_l}
      boxExtents=<10 10 10>
      constraintSetup {
       angularLimitsMax=<10 15 25>
       angularLimitsMin=<-10 -15 -25>
       angularTarget=<0 0 0>
       linearAxesMax=<0.5 0.5 0.5>
       linearAxesMin=<-0.5 -0.5 -0.5>
      }
      localJointOffset=<1 0 0>
     }]
    relativeSpaceBone {boneName=spine_03}
   }
   n16 = Anim Dynamics - Bone: dyn_upperarm_01_l(ComponentPose=@n15, Alpha=1, ExternalForce=@n14) {
    angularDampingOverride=1
    angularSpringConstant=50
    bGravityOverrideInSimSpace=true
    bLinearSpring=true
    bOverrideAngularDamping=true
    bUseGravityOverride=true
    bUsePlanarLimit=false
    gravityScale=0.25
    linearSpringConstant=25
    simulationSpace=RootRelative
    alphaBoolBlend {blendOption=Linear}
    boundBone {boneName=dyn_upperarm_01_l}
    chainEnd {boneName=root}
    componentAppliedLinearAccClamp=<3000 3000 3000>
    componentLinearAccScale=<0.7 0.7 0.7>
    componentLinearVelScale=<0.7 0.7 0.7>
    gravityOverride {z=-980}
    physicsBodyDefinitions [{
      sphereCollisionRadius=1
      boundBone {boneName=dyn_upperarm_01_l}
      boxExtents=<10 10 10>
      constraintSetup {
       angularLimitsMax=<10 15 25>
       angularLimitsMin=<-10 -15 -25>
       angularTarget=<0 0 0>
       linearAxesMax=<0.5 0.5 0.5>
       linearAxesMin=<-0.5 -0.5 -0.5>
      }
      localJointOffset=<1 0 0>
     }]
    relativeSpaceBone {boneName=spine_03}
   }
   n17 = Anim Dynamics - Bone: dyn_upperarm_02_r(ComponentPose=@n16, Alpha=1, ExternalForce=@n14) {
    angularDampingOverride=1
    angularSpringConstant=0.1
    bAngularSpring=true
    bGravityOverrideInSimSpace=true
    bLinearSpring=true
    bOverrideAngularDamping=true
    bUseGravityOverride=true
    bUsePlanarLimit=false
    gravityScale=0.25
    linearDampingOverride=1
    linearSpringConstant=0.1
    simulationSpace=RootRelative
    alphaBoolBlend {blendOption=Linear}
    boundBone {boneName=dyn_upperarm_02_r}
    chainEnd {boneName=root}
    componentAppliedLinearAccClamp=<3000 3000 3000>
    componentLinearAccScale=<0.7 0.7 0.7>
    componentLinearVelScale=<0.7 0.7 0.7>
    gravityOverride {z=-100}
    physicsBodyDefinitions [{
      sphereCollisionRadius=1
      boundBone {boneName=dyn_upperarm_02_r}
      boxExtents=<10 10 10>
      constraintSetup {angularLimitsMax=<2 2 2>, angularLimitsMin=<-2 -2 -2>, angularTarget=<-1 0 0>, linearAxesMax=<0.5 0.5 0.5>, linearAxesMin=<-0.5 -0.5 -0.5>}
      localJointOffset=<5 0 0>
     }]
    relativeSpaceBone {boneName=spine_03}
   }
   n18 = Anim Dynamics - Bone: dyn_upperarm_01_r(ComponentPose=@n17, Alpha=1, ExternalForce=@n14) {
    angularDampingOverride=1
    angularSpringConstant=0.1
    bAngularSpring=true
    bGravityOverrideInSimSpace=true
    bLinearSpring=true
    bOverrideAngularDamping=true
    bUseGravityOverride=true
    bUsePlanarLimit=false
    gravityScale=0.25
    linearDampingOverride=1
    linearSpringConstant=0.1
    simulationSpace=RootRelative
    alphaBoolBlend {blendOption=Linear}
    boundBone {boneName=dyn_upperarm_01_r}
    chainEnd {boneName=root}
    componentAppliedLinearAccClamp=<3000 3000 3000>
    componentLinearAccScale=<0.7 0.7 0.7>
    componentLinearVelScale=<0.7 0.7 0.7>
    gravityOverride {z=-100}
    physicsBodyDefinitions [{
      sphereCollisionRadius=1
      boundBone {boneName=dyn_upperarm_01_r}
      boxExtents=<10 10 10>
      constraintSetup {angularLimitsMax=<2 2 2>, angularLimitsMin=<-2 -2 -2>, angularTarget=<-1 0 0>, linearAxesMax=<0.5 0.5 0.5>, linearAxesMin=<-0.5 -0.5 -0.5>}
      localJointOffset=<5 0 0>
     }]
    relativeSpaceBone {boneName=spine_03}
   }
   n19 = Get isTargetingAlpha() -> (isTargetingAlpha:double)
   n20 = "Transform (Modify) Bone - Bone: dyn_jacket_mid_01_r_MSTR"(ComponentPose=@n18, Alpha=@n19, Translation="0,0,0", Rotation="0,-30,0", Scale="1,1,1") {rotationMode=BMM_Additive, alphaBoolBlend {blendOption=Linear}, boneToModify {boneName=dyn_jacket_mid_01_r_MSTR}, rotation {yaw=-30}}
   n21 = "Transform (Modify) Bone - Bone: dyn_jacket_upper_01_r_MSTR"(ComponentPose=@n20, Alpha=@n19, Translation="0,0,0", Rotation="0,-75,0", Scale="1,1,1") {rotationMode=BMM_Additive, alphaBoolBlend {blendOption=Linear}, boneToModify {boneName=dyn_jacket_upper_01_r_MSTR}, rotation {yaw=-75}}
   n22 = "Transform (Modify) Bone - Bone: upperarm_armor_01_l_DSP"(ComponentPose=@n21, Alpha=@n19, Translation="0,0,0", Rotation="-15,0,0", Scale="1,1,1") {rotationMode=BMM_Additive, alphaBoolBlend {blendOption=Linear}, boneToModify {boneName=upperarm_armor_01_l_DSP}, rotation {pitch=-15}}
   n23 = "Transform (Modify) Bone - Bone: dyn_collar_01_l_MSTR"(ComponentPose=@n22, Alpha=Get armSwingDelta_r() -> (armSwingDelta_r:double), Translation="0,0,0", Rotation="35,0,0", Scale="1,1,1") {rotationMode=BMM_Additive, alphaBoolBlend {blendOption=Linear}, boneToModify {boneName=dyn_collar_01_l_MSTR}, rotation {pitch=35}}
   n24 = "Transform (Modify) Bone - Bone: dyn_collar_01_r_MSTR"(ComponentPose=@n23, Alpha=@n1, Translation="2,0,0", Rotation="0,-60,0", Scale="1,1,1") {
    rotationMode=BMM_Additive
    translationMode=BMM_Additive
    alphaBoolBlend {blendOption=Linear}
    boneToModify {boneName=dyn_collar_01_r_MSTR}
    rotation {yaw=-60}
    translation {x=2}
   }
   n25 = "Transform (Modify) Bone - Bone: upperarm_armor_01_r_DSP"(ComponentPose=@n24, Alpha=@n1, Translation="0,0,0", Rotation="0,-15,0", Scale="1,1,1") {
    rotationMode=BMM_Additive
    rotationSpace=BCS_BoneSpace
    alphaBoolBlend {blendOption=Linear}
    boneToModify {boneName=upperarm_armor_01_r_DSP}
    rotation {yaw=-15}
   }
   n26 = "Transform (Modify) Bone - Bone: dyn_collar_strap_01_r"(ComponentPose=@n25, Alpha=@n1, Translation="0,0,0", Rotation="0,-20,0", Scale="1,1,1") {
    rotationMode=BMM_Additive
    translationMode=BMM_Additive
    alphaBoolBlend {blendOption=Linear}
    boneToModify {boneName=dyn_collar_strap_01_r}
    rotation {yaw=-20}
   }
   n27 = "Transform (Modify) Bone - Bone: dyn_collar_strap_02_r"(ComponentPose=@n26, Alpha=@n1, Translation="0,0,0", Rotation="0,-20,0", Scale="1,1,1") {
    rotationMode=BMM_Additive
    translationMode=BMM_Additive
    alphaBoolBlend {blendOption=Linear}
    boneToModify {boneName=dyn_collar_strap_02_r}
    rotation {yaw=-20}
   }
   n28 = "Transform (Modify) Bone - Bone: dyn_jacket_mid_01_r_MSTR"(ComponentPose=@n27, Alpha=@n1, Translation="-0.5,0,0", Rotation="0,45,0", Scale="1,1,1") {
    rotationMode=BMM_Additive
    translationMode=BMM_Additive
    alphaBoolBlend {blendOption=Linear}
    boneToModify {boneName=dyn_jacket_mid_01_r_MSTR}
    rotation {yaw=45}
    translation {x=-0.5}
   }
   n29 = "Transform (Modify) Bone - Bone: dyn_boot_01_r_MSTR"(
    ComponentPose=@n28
    Alpha=Get isCrouchingNotWalkingAlpha() -> (isCrouchingNotWalkingAlpha:double)
    Translation="0,0,0"
    Rotation="0,0,-50"
    Scale="1,1,1"
   ) {rotationMode=BMM_Additive, alphaBoolBlend {blendOption=Linear}, boneToModify {boneName=dyn_boot_01_r_MSTR}, rotation {roll=-50}}
   n30 = "Constraint - knee_armor_02_l_DSP (RR)"!(ComponentPose=@n29, Alpha=1, ConstraintWeights_0=0.75, ConstraintWeights_1:float=0.25) -> (Pose) {
    alphaBoolBlend {blendOption=Linear}
    boneToModify {boneName=knee_armor_02_l_DSP}
    constraintSetup [{transformType=Rotation, targetBone {boneName=calf_l}}, {transformType=Rotation, targetBone {boneName=thigh_l}}]
    constraintWeights [1,1]
   }
   n31 = Get backpackAlpha() -> (backpackAlpha:double)
   n32 = "Constraint - dyn_jacket_back_01 (P)"(ComponentPose=@n30, Alpha=@n31, ConstraintWeights_0=1) {
    alphaBoolBlend {blendOption=Linear}
    boneToModify {boneName=dyn_jacket_back_01}
    constraintSetup [{targetBone {boneName=spine_04}}]
    constraintWeights [1]
   }
   n33 = "Constraint - dyn_jacket_back_02 (P)"(ComponentPose=@n32, Alpha=@n31, ConstraintWeights_0=1) {
    alphaBoolBlend {blendOption=Linear}
    boneToModify {boneName=dyn_jacket_back_02}
    constraintSetup [{targetBone {boneName=spine_04}}]
    constraintWeights [1]
   }
   n34 = "Constraint - dyn_jacket_side_01_l (P)"(ComponentPose=@n33, Alpha=@n31, ConstraintWeights_0=1) {
    alphaBoolBlend {blendOption=Linear}
    boneToModify {boneName=dyn_jacket_side_01_l}
    constraintSetup [{targetBone {boneName=spine_04}}]
    constraintWeights [1]
   }
   n35 = "Constraint - dyn_jacket_side_02_l (P)"(ComponentPose=@n34, Alpha=@n31, ConstraintWeights_0=1) {
    alphaBoolBlend {blendOption=Linear}
    boneToModify {boneName=dyn_jacket_side_02_l}
    constraintSetup [{targetBone {boneName=spine_04}}]
    constraintWeights [1]
   }
   n36 = "Constraint - dyn_jacket_side_01_r (P)"(ComponentPose=@n35, Alpha=@n31, ConstraintWeights_0=1) {
    alphaBoolBlend {blendOption=Linear}
    boneToModify {boneName=dyn_jacket_side_01_r}
    constraintSetup [{targetBone {boneName=spine_04}}]
    constraintWeights [1]
   }
   n37 = "Constraint - dyn_jacket_side_02_r (P)"(ComponentPose=@n36, Alpha=@n31, ConstraintWeights_0=1) {
    alphaBoolBlend {blendOption=Linear}
    boneToModify {boneName=dyn_jacket_side_02_r}
    constraintSetup [{targetBone {boneName=spine_04}}]
    constraintWeights [1]
   }
   Output Pose(Result=Component To Local(ComponentPose=@n37)) {layerGroup=None}
  ]}
 EventGraph {nodes [
   Event BlueprintUpdateAnimation()
   > n1 = Sequence!(execute) -> (then_0, then_1, then_2:Exec, then_3:Exec, then_4:Exec, then_5:Exec)
   n2 = NOT Boolean(A=Get bIsMoving2D())
   n3 = Get bIsCrouching()
   > Set isCrouchingNotWalkingAlpha(isCrouchingNotWalkingAlpha:double="To Float (Boolean)"(InBool=AND Boolean(A=@n2, B=@n3))) -> (Output_Get:double)
   n4 = Get bRecentlyFired() -> (bRecentlyFired:bool)
   Set shootingForce(execute=@n1.then_1, shootingForce:Vector=MakeVector(Y=0, Z=SelectFloat(A=2800, B=0, bPickA=@n4))) -> (Output_Get:Vector)
   Set aimUpAlpha(execute=@n1.then_2, aimUpAlpha:double=MapRangeClamped(Value=Get AimPitch(), InRangeA=0, InRangeB=90)) -> (Output_Get:double)
   Set aimPitchAlpha(execute=@n1.then_3, aimPitchAlpha:double=Get AimPitch()) -> (Output_Get:double)
   n5 = GetOwnerSkeletalMeshComponent()
   n6 = BreakTransform(InTransform=GetDeltaTransformFromRefPose(self=@n5, BoneName=upperarm_l, BaseName=clavicle_l))
   n7 = MapRangeClamped(Value="Absolute (Float)"(A=BreakRotator(InRot=@n6.Rotation).Yaw), InRangeA=5, InRangeB=70)
   Set armSwingDelta_r(execute=@n1.then_5, armSwingDelta_r:double=@n7) -> (Output_Get:double)
   > Set isAimingUpAlpha(isAimingUpAlpha:double=SelectFloat(A=@n7, B=0, bPickA=Get bIsTargeting())) -> (Output_Get:double)
   n8 = BreakTransform(InTransform=GetDeltaTransformFromRefPose(self=@n5, BoneName=head, BaseName=spine_05))
   n9 = Get bIsSkyDiving()
   > Set headSwingAlpha(headSwingAlpha:double=MapRangeClamped(Value=BreakRotator(InRot=@n8.Rotation).Yaw, InRangeA=-45, InRangeB=0, OutRangeA=SelectFloat(A=1, B=0.75, bPickA=@n9), OutRangeB=0)) -> (Output_Get:double)
   Event BlueprintInitializeAnimation()
   n10 = TryGetPawnOwner()
   > Is Valid(InputObject=@n10)
   > n11 = Cast To FortPlayerPawnAthena(Object=@n10)
   Event OnEndTargeting() {is=title K2Node_Event}
   > Set isTargetingAlpha(isTargetingAlpha:double=0) -> (Output_Get:double)
   > Set isTargetingHalf(isTargetingHalf:double=0) -> (Output_Get:double)
   Event OnBeginCrouching()
   > Set isCrouchingAlpha(isCrouchingAlpha:double=1) -> (Output_Get:double)
   Event OnEndCrouching()
   > Set isCrouchingAlpha(isCrouchingAlpha=0)
   Event OnBeginTargeting() {is=title K2Node_Event}
   > Set isTargetingAlpha(isTargetingAlpha=1)
   > Set isTargetingHalf(isTargetingHalf=0.5)
   Event OnCharacterCustomizationLoaded()
   > Is Valid(InputObject=@n11)
   > Do Once()
   n12 = GetSkeletalMeshForPartType(self=@n11, PartType=Backpack)
   > Set backpackAlpha(backpackAlpha:double="To Float (Boolean)"(InBool="Not Equal (Object)"!(A:Object Reference=@n12, B:Object Reference) -> (ReturnValue:bool))) -> (Output_Get:double)
   n13 = CheckForBackbling()
   OR Boolean(A=@n13, B=@n13.bIsBackblingLong)
  ]}
]
variables [
 aimPitchAlpha {type=number}
 aimUpAlpha {type=number}
 armSwingDelta_r {type=number}
 backpackAlpha {type=number}
 headSwingAlpha {type=number}
 isAimingUpAlpha {type=number}
 isCrouchingAlpha {type=number}
 isCrouchingNotWalkingAlpha {type=number}
 isTargetingAlpha {type=number}
 isTargetingHalf {type=number}
 shootingForce {type=Vector}
]
