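# Physics asset for the front hair strand face accessory on the female medium Future Biker player body, simulating the dynamic left-side strand bones.
# A head capsule anchors a tiny master box with free swing and twist, followed by box bodies on strands 01 and 02.
# The chain forks at strand 02 into strands 03 and 05, whose constraints lock twist.
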PhysicsAsset /BRCosmetics/Characters/Player/Female/Medium/Bodies/F_MED_Future_Biker/Meshes/Parts/F_MED_Future_Biker_FaceAcc_FrontStrands_Physics
bodies {
 dyn_hairstrand_01_l [{shapeType=Box, extentX=4.6, extentY=5.6, extentZ=2.6, center=<2.0292 0 0.0001>}]
 dyn_hairstrand_01_l_MSTR [{shapeType=Box, extentX=0.01, extentY=0.01, extentZ=1.1}]
 dyn_hairstrand_02_l [{shapeType=Box, extentX=4.6, extentY=5.6, extentZ=2.1, center=<2.0292 0 0.0001>}]
 dyn_hairstrand_03_l [{shapeType=Box, extentX=4.6, extentY=5.6, extentZ=2.1, center=<2.0292 0 0.0001>}]
 dyn_hairstrand_05_l [{shapeType=Box, extentX=4.6, extentY=5.6, extentZ=2.1, center=<2.0292 0 0.0001>}]
 head [{shapeType=Capsule, length=5.8508, radius=7.0443, center=<2.7081 1.4901 0>, rotation=<rot 90 0 -10.0006>}]
}
constraints [
 {
  bone1Name=dyn_hairstrand_01_l_MSTR
  bone2Name=head
  swing1LimitDegrees=45
  swing1Motion=Free
  swing2LimitDegrees=45
  swing2Motion=Free
  twistLimitDegrees=45
  twistMotion=Free
 }
 {bone1Name=dyn_hairstrand_01_l, bone2Name=dyn_hairstrand_01_l_MSTR, swing1LimitDegrees=30, swing2LimitDegrees=45, twistLimitDegrees=0}
 {bone1Name=dyn_hairstrand_02_l, bone2Name=dyn_hairstrand_01_l, swing1LimitDegrees=30, swing2LimitDegrees=30, twistLimitDegrees=0}
 {
  bone1Name=dyn_hairstrand_05_l
  bone2Name=dyn_hairstrand_02_l
  swing1LimitDegrees=45
  swing2LimitDegrees=45
  twistLimitDegrees=5
  twistMotion=Locked
 }
 {
  bone1Name=dyn_hairstrand_03_l
  bone2Name=dyn_hairstrand_02_l
  swing1LimitDegrees=45
  swing2LimitDegrees=45
  twistLimitDegrees=5
  twistMotion=Locked
 }
]
