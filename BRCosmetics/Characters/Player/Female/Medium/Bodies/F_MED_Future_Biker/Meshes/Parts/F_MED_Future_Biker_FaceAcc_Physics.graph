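# Physics asset for the female Future Biker face accessory mesh, defining bodies and joints for its dynamic hair bones around the head.
# Twelve back-hair pieces, a single strand and two-segment left front and back strands use boxes, with one capsule for the head.
# Each chain hangs from a tiny MSTR anchor body with free rotation against the head.
# Segments below the anchors use angular limits with narrow twist, and the left back strand also narrows one swing axis.
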PhysicsAsset /BRCosmetics/Characters/Player/Female/Medium/Bodies/F_MED_Future_Biker/Meshes/Parts/F_MED_Future_Biker_FaceAcc_Physics
bodies {
 dyn_backhair10_01 [{shapeType=Box, extentX=6.1, extentY=4.1, extentZ=1.1, center=<3.1441 0 0.0001>}]
 dyn_backhair10_01_MSTR [{shapeType=Box, extentX=0.01, extentY=0.01, extentZ=1.1}]
 dyn_backhair11_01 [{shapeType=Box, extentX=6.1, extentY=4.1, extentZ=1.1, center=<3.1441 0 0.0001>}]
 dyn_backhair11_01_MSTR [{shapeType=Box, extentX=0.01, extentY=0.01, extentZ=1.1}]
 dyn_backhair12_01 [{shapeType=Box, extentX=6.1, extentY=4.1, extentZ=1.1, center=<3.1441 0 0.0001>}]
 dyn_backhair12_01_MSTR [{shapeType=Box, extentX=0.01, extentY=0.01, extentZ=1.1}]
 dyn_backhair1_01 [{shapeType=Box, extentX=6.1, extentY=4.1, extentZ=1.1, center=<3.1441 0 0.0001>}]
 dyn_backhair1_01_MSTR [{shapeType=Box, extentX=0.01, extentY=0.01, extentZ=1.1}]
 dyn_backhair2_01 [{shapeType=Box, extentX=6.1, extentY=4.1, extentZ=1.1, center=<3.1441 0 0.0001>}]
 dyn_backhair2_01_MSTR [{shapeType=Box, extentX=0.01, extentY=0.01, extentZ=1.1}]
 dyn_backhair3_01 [{shapeType=Box, extentX=6.1, extentY=4.1, extentZ=1.1, center=<3.1441 0 0.0001>}]
 dyn_backhair3_01_MSTR [{shapeType=Box, extentX=0.01, extentY=0.01, extentZ=1.1}]
 dyn_backhair4_01 [{shapeType=Box, extentX=6.1, extentY=4.1, extentZ=1.1, center=<3.1441 0 0.0001>}]
 dyn_backhair4_01_MSTR [{shapeType=Box, extentX=0.01, extentY=0.01, extentZ=1.1}]
 dyn_backhair5_01 [{shapeType=Box, extentX=6.1, extentY=4.1, extentZ=1.1, center=<3.1441 0 0.0001>}]
 dyn_backhair5_01_MSTR [{shapeType=Box, extentX=0.01, extentY=0.01, extentZ=1.1}]
 dyn_backhair6_01 [{shapeType=Box, extentX=6.1, extentY=4.1, extentZ=1.1, center=<3.1441 0 0.0001>}]
 dyn_backhair6_01_MSTR [{shapeType=Box, extentX=0.01, extentY=0.01, extentZ=1.1}]
 dyn_backhair7_01 [{shapeType=Box, extentX=6.1, extentY=4.1, extentZ=1.1, center=<3.1441 0 0.0001>}]
 dyn_backhair7_01_MSTR [{shapeType=Box, extentX=0.01, extentY=0.01, extentZ=1.1}]
 dyn_backhair8_01 [{shapeType=Box, extentX=6.1, extentY=4.1, extentZ=1.1, center=<3.1441 0 0.0001>}]
 dyn_backhair8_01_MSTR [{shapeType=Box, extentX=0.01, extentY=0.01, extentZ=1.1}]
 dyn_backhair9_01 [{shapeType=Box, extentX=6.1, extentY=4.1, extentZ=1.1, center=<3.1441 0 0.0001>}]
 dyn_backhair9_01_MSTR [{shapeType=Box, extentX=0.01, extentY=0.01, extentZ=1.1}]
 dyn_hairstrand_01 [{shapeType=Box, extentX=6.6, extentY=4.1, extentZ=2.6, center=<3.336 0 0.0001>}]
 dyn_hairstrand_01_MSTR [{shapeType=Box, extentX=0.01, extentY=0.01, extentZ=1.1}]
 dyn_hairstrand_back_01_l [{shapeType=Box, extentX=4.6, extentY=5.6, extentZ=2.6, center=<2.0292 0 0.0001>}]
 dyn_hairstrand_back_01_l_MSTR [{shapeType=Box, extentX=0.01, extentY=0.01, extentZ=1.1}]
 dyn_hairstrand_back_02_l [{shapeType=Box, extentX=4.6, extentY=5.6, extentZ=2.6, center=<2.0292 0 0.0001>}]
 dyn_hairstrand_front_01_l [{shapeType=Box, extentX=4.6, extentY=5.6, extentZ=2.1164, center=<2.0292 0 0.0001>}]
 dyn_hairstrand_front_01_l_MSTR [{shapeType=Box, extentX=0.01, extentY=0.01, extentZ=1.1}]
 dyn_hairstrand_front_02_l [{shapeType=Box, extentX=4.6, extentY=5.6, extentZ=1.593, center=<2.0292 0 0.0001>}]
 head [{shapeType=Capsule, length=5.8508, radius=7.0443, center=<2.7081 1.4901 0>, rotation=<rot 90 0 -10.0006>}]
}
constraints [
 {
  bone1Name=dyn_backhair2_01_MSTR
  bone2Name=head
  swing1LimitDegrees=45
  swing1Motion=Free
  swing2LimitDegrees=45
  swing2Motion=Free
  twistLimitDegrees=45
  twistMotion=Free
 }
 {bone1Name=dyn_backhair2_01, bone2Name=dyn_backhair2_01_MSTR, swing1LimitDegrees=45, swing2LimitDegrees=45, twistLimitDegrees=5}
 {
  bone1Name=dyn_backhair1_01_MSTR
  bone2Name=head
  swing1LimitDegrees=45
  swing1Motion=Free
  swing2LimitDegrees=45
  swing2Motion=Free
  twistLimitDegrees=45
  twistMotion=Free
 }
 {bone1Name=dyn_backhair1_01, bone2Name=dyn_backhair1_01_MSTR, swing1LimitDegrees=45, swing2LimitDegrees=45, twistLimitDegrees=5}
 {
  bone1Name=dyn_backhair3_01_MSTR
  bone2Name=head
  swing1LimitDegrees=45
  swing1Motion=Free
  swing2LimitDegrees=45
  swing2Motion=Free
  twistLimitDegrees=45
  twistMotion=Free
 }
 {bone1Name=dyn_backhair3_01, bone2Name=dyn_backhair3_01_MSTR, swing1LimitDegrees=45, swing2LimitDegrees=45, twistLimitDegrees=5}
 {
  bone1Name=dyn_backhair4_01_MSTR
  bone2Name=head
  swing1LimitDegrees=45
  swing1Motion=Free
  swing2LimitDegrees=45
  swing2Motion=Free
  twistLimitDegrees=45
  twistMotion=Free
 }
 {bone1Name=dyn_backhair4_01, bone2Name=dyn_backhair4_01_MSTR, swing1LimitDegrees=45, swing2LimitDegrees=45, twistLimitDegrees=5}
 {
  bone1Name=dyn_hairstrand_back_01_l_MSTR
  bone2Name=head
  swing1LimitDegrees=45
  swing1Motion=Free
  swing2LimitDegrees=45
  swing2Motion=Free
  twistLimitDegrees=45
  twistMotion=Free
 }
 {
  bone1Name=dyn_hairstrand_back_01_l
  bone2Name=dyn_hairstrand_back_01_l_MSTR
  swing1LimitDegrees=15
  swing2LimitDegrees=45
  twistLimitDegrees=5
 }
 {bone1Name=dyn_hairstrand_back_02_l, bone2Name=dyn_hairstrand_back_01_l, swing1LimitDegrees=15, swing2LimitDegrees=45, twistLimitDegrees=5}
 {
  bone1Name=dyn_hairstrand_front_01_l_MSTR
  bone2Name=head
  swing1LimitDegrees=45
  swing1Motion=Free
  swing2LimitDegrees=45
  swing2Motion=Free
  twistLimitDegrees=45
  twistMotion=Free
 }
 {
  bone1Name=dyn_hairstrand_front_01_l
  bone2Name=dyn_hairstrand_front_01_l_MSTR
  swing1LimitDegrees=45
  swing2LimitDegrees=45
  twistLimitDegrees=5
 }
 {bone1Name=dyn_hairstrand_front_02_l, bone2Name=dyn_hairstrand_front_01_l, swing1LimitDegrees=45, swing2LimitDegrees=45, twistLimitDegrees=5}
 {
  bone1Name=dyn_hairstrand_01_MSTR
  bone2Name=head
  swing1LimitDegrees=45
  swing1Motion=Free
  swing2LimitDegrees=45
  swing2Motion=Free
  twistLimitDegrees=45
  twistMotion=Free
 }
 {bone1Name=dyn_hairstrand_01, bone2Name=dyn_hairstrand_01_MSTR, swing1LimitDegrees=45, swing2LimitDegrees=45, twistLimitDegrees=5}
 {
  bone1Name=dyn_backhair12_01_MSTR
  bone2Name=head
  swing1LimitDegrees=45
  swing1Motion=Free
  swing2LimitDegrees=45
  swing2Motion=Free
  twistLimitDegrees=45
  twistMotion=Free
 }
 {bone1Name=dyn_backhair12_01, bone2Name=dyn_backhair12_01_MSTR, swing1LimitDegrees=45, swing2LimitDegrees=45, twistLimitDegrees=5}
 {
  bone1Name=dyn_backhair11_01_MSTR
  bone2Name=head
  swing1LimitDegrees=45
  swing1Motion=Free
  swing2LimitDegrees=45
  swing2Motion=Free
  twistLimitDegrees=45
  twistMotion=Free
 }
 {bone1Name=dyn_backhair11_01, bone2Name=dyn_backhair11_01_MSTR, swing1LimitDegrees=45, swing2LimitDegrees=45, twistLimitDegrees=5}
 {
  bone1Name=dyn_backhair9_01_MSTR
  bone2Name=head
  swing1LimitDegrees=45
  swing1Motion=Free
  swing2LimitDegrees=45
  swing2Motion=Free
  twistLimitDegrees=45
  twistMotion=Free
 }
 {bone1Name=dyn_backhair9_01, bone2Name=dyn_backhair9_01_MSTR, swing1LimitDegrees=45, swing2LimitDegrees=45, twistLimitDegrees=5}
 {
  bone1Name=dyn_backhair10_01_MSTR
  bone2Name=head
  swing1LimitDegrees=45
  swing1Motion=Free
  swing2LimitDegrees=45
  swing2Motion=Free
  twistLimitDegrees=45
  twistMotion=Free
 }
 {bone1Name=dyn_backhair10_01, bone2Name=dyn_backhair10_01_MSTR, swing1LimitDegrees=45, swing2LimitDegrees=45, twistLimitDegrees=5}
 {
  bone1Name=dyn_backhair6_01_MSTR
  bone2Name=head
  swing1LimitDegrees=45
  swing1Motion=Free
  swing2LimitDegrees=45
  swing2Motion=Free
  twistLimitDegrees=45
  twistMotion=Free
 }
 {bone1Name=dyn_backhair6_01, bone2Name=dyn_backhair6_01_MSTR, swing1LimitDegrees=45, swing2LimitDegrees=45, twistLimitDegrees=5}
 {
  bone1Name=dyn_backhair5_01_MSTR
  bone2Name=head
  swing1LimitDegrees=45
  swing1Motion=Free
  swing2LimitDegrees=45
  swing2Motion=Free
  twistLimitDegrees=45
  twistMotion=Free
 }
 {bone1Name=dyn_backhair5_01, bone2Name=dyn_backhair5_01_MSTR, swing1LimitDegrees=45, swing2LimitDegrees=45, twistLimitDegrees=5}
 {
  bone1Name=dyn_backhair7_01_MSTR
  bone2Name=head
  swing1LimitDegrees=45
  swing1Motion=Free
  swing2LimitDegrees=45
  swing2Motion=Free
  twistLimitDegrees=45
  twistMotion=Free
 }
 {bone1Name=dyn_backhair7_01, bone2Name=dyn_backhair7_01_MSTR, swing1LimitDegrees=45, swing2LimitDegrees=45, twistLimitDegrees=5}
 {
  bone1Name=dyn_backhair8_01_MSTR
  bone2Name=head
  swing1LimitDegrees=45
  swing1Motion=Free
  swing2LimitDegrees=45
  swing2Motion=Free
  twistLimitDegrees=45
  twistMotion=Free
 }
 {bone1Name=dyn_backhair8_01, bone2Name=dyn_backhair8_01_MSTR, swing1LimitDegrees=45, swing2LimitDegrees=45, twistLimitDegrees=5}
]
