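# Post-copy physics AnimBP for the Future Biker female FaceAcc mesh, adding secondary motion to hair and face accessory bones.
# Copies the parent mesh pose, then runs head-space RigidBody sims for the main and front-strand physics assets, both fed skydiving external force.
# During emotes or custom animations, EventGraph alphas swap the front-strand RigidBody for spring-based AnimDynamics on two left hair strand bones.
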AnimBlueprint /BRCosmetics/Characters/Player/Female/Medium/Bodies/F_MED_Future_Biker/Meshes/Parts/F_MED_Future_Biker_FaceAcc_Skeleton_AnimBP
target_skeleton=&/F_MED_Future_Biker_FaceAcc_Skeleton
graphs [
 AnimGraph {nodes [
   n1 = Break Fort Anim Input Skydiving External Force(FortAnimInput_SkydivingExternalForce=Get SkydivingExternalForceInput()) -> (
    bUseSkydivingVectorForce=~
    SkydivingVectorForce=~
    HeadToPelvisDirection=~
    FloatingMultiplier=~
    FloatingAdditive=~
    DivingMultiplier=~
    DivingAdditive=~
    ParachutingMultiplier=~
    ParachutingAdditive=~
   )
   n2 = RigidBody(ComponentPose=Local To Component(LocalPose=Copy Pose From Mesh() {bUseAttachedParent=true}), ExternalForce=@n1.SkydivingVectorForce) {
    bClampLinearTranslationLimitToRefPose=true
    bOverrideWorldGravity=true
    overridePhysicsAsset=&/F_MED_Future_Biker_FaceAcc_Physics
    simulationSpace=BaseBoneSpace
    alphaBoolBlend {blendOption=Linear}
    baseBoneRef {boneName=head}
    componentAppliedLinearAccClamp=<1500 1500 1500>
    componentLinearAccScale=<0.3 0.3 0.6>
    componentLinearVelScale=<0.6 0.6 2>
    overrideWorldGravity {z=-980}
   }
   n3 = RigidBody(ComponentPose=@n2, Alpha=Get isEmotingAlpha_inv() -> (isEmotingAlpha_inv:double), ExternalForce=@n1.SkydivingVectorForce) {
    bClampLinearTranslationLimitToRefPose=true
    bOverrideWorldGravity=true
    overridePhysicsAsset=&/F_MED_Future_Biker_FaceAcc_FrontStrands_Physics
    simulationSpace=BaseBoneSpace
    alphaBoolBlend {blendOption=Linear}
    baseBoneRef {boneName=head}
    componentAppliedLinearAccClamp=<1500 1500 1500>
    componentLinearAccScale=<0.3 0.3 0.6>
    componentLinearVelScale=<0.6 0.6 2>
    overrideWorldGravity {z=-980}
   }
   n4 = Get isEmotingAlpha() -> (isEmotingAlpha:double)
   n5 = Anim Dynamics - Bone: dyn_hairstrand_01_l(ComponentPose=@n3, Alpha=@n4, ExternalForce="0,0,0") {
    angularDampingOverride=1
    angularSpringConstant=100
    bAngularSpring=true
    bGravityOverrideInSimSpace=true
    bOverrideAngularDamping=true
    bUseGravityOverride=true
    bUsePlanarLimit=false
    gravityScale=0.25
    alphaBoolBlend {blendOption=Linear}
    boundBone {boneName=dyn_hairstrand_01_l}
    componentAppliedLinearAccClamp=<3000 3000 3000>
    componentLinearAccScale=<0.7 0.7 0.7>
    componentLinearVelScale=<0.7 0.7 0.7>
    gravityOverride {z=-980}
    physicsBodyDefinitions [{
      sphereCollisionRadius=1
      boundBone {boneName=dyn_hairstrand_01_l}
      boxExtents=<10 10 10>
      constraintSetup {angularLimitsMax=<5 0 10>, angularLimitsMin=<-5 -20 -5>, angularTarget=<0 0 0>}
      localJointOffset=<3 0 0>
     }]
    relativeSpaceBone {boneName=head}
   }
   n6 = Anim Dynamics - Bone: dyn_hairstrand_02_l(ComponentPose=@n5, Alpha=@n4, ExternalForce="0,0,0") {
    angularDampingOverride=1
    angularSpringConstant=100
    bAngularSpring=true
    bGravityOverrideInSimSpace=true
    bOverrideAngularDamping=true
    bUseGravityOverride=true
    bUsePlanarLimit=false
    gravityScale=0.25
    alphaBoolBlend {blendOption=Linear}
    boundBone {boneName=dyn_hairstrand_02_l}
    componentAppliedLinearAccClamp=<3000 3000 3000>
    componentLinearAccScale=<0.7 0.7 0.7>
    componentLinearVelScale=<0.7 0.7 0.7>
    gravityOverride {z=-980}
    physicsBodyDefinitions [{
      sphereCollisionRadius=1
      boundBone {boneName=dyn_hairstrand_02_l}
      boxExtents=<10 10 10>
      constraintSetup {angularLimitsMax=<5 0 10>, angularLimitsMin=<-5 -20 -5>, angularTarget=<0 0 0>}
      localJointOffset=<3 0 0>
     }]
    relativeSpaceBone {boneName=head}
   }
   Output Pose(Result=Component To Local(ComponentPose=@n6)) {layerGroup=None}
  ]}
 EventGraph {nodes [
   Event BlueprintUpdateAnimation()
   n1 = OR Boolean(A=Get bCanPlayCustomAnimations() -> (bCanPlayCustomAnimations:bool), B=Get bIsPlayingEmote())
   > Set isEmotingAlpha(isEmotingAlpha:double="To Float (Boolean)"(InBool=@n1)) -> (Output_Get:double)
   > Set isEmotingAlpha_inv(isEmotingAlpha_inv:double="To Float (Boolean)"(InBool=NOT Boolean(A=@n1))) -> (Output_Get:double)
  ]}
]
variables [
 isEmotingAlpha {type=number}
 isEmotingAlpha_inv {type=number}
]
